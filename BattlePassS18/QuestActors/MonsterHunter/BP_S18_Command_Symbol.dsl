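; Monster Hunter quest actor that coordinates a group of Control Torches and reveals a Command Symbol when the player activates them.
; On begin play the server registers itself as each torch's controller and sets the required activation-effect stacks to the torch count.
; Entering range grants a listener effect server-side and lights torches client-side for the local quest player; leaving undoes both and clears progress.
; EvaluateSymbolReveal compares stack counts, then applies the revealed effect and checks quest completion.
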(event EventBeginPlay
  (|CallParentFunction)
  (Networking|SwitchHasAuthority
    (:Authority
      (Utilities|Array|Clear (Variables|Data|GetTouchPlayerArray))
      (Development|PrintString "@@@ Begin Play [BP_S18_Command_Symbol]: CALLED!" false (Variables|Debug|GetbDebug))
      (CallFunction|InitializeCommandSymbol (Variables|Data|GetControlTorchArray)))))

(event EventPlayerInInteractRange
  (params (Actor :type Actor%20Object%20Reference))
  (|CallParentFunction Actor)
  (Networking|SwitchHasAuthority
    (:Authority
      (CallFunction|ProcessPlayerBeginOverlap Actor))
    (:Remote
      (Development|PrintString "@@@@@@ In Interact Range [BP_S18_Command_Symbol]: CALLED!" false (Variables|Debug|GetbDebug))
      (if (== (Utilities|Array|FindItem (Variables|Data|GetTouchPlayerArray) Actor) -1)
        (bind _asfort_player_pawn (Utilities|Casting|CastToFortPlayerPawn Actor))
        (if (Pawn|IsLocallyControlled _asfort_player_pawn)
          (Utilities|Array|AddUnique (Variables|Data|GetTouchPlayerArray) Actor)
          (bind _binprogress (CallFunction|CheckQuestProgress _asfort_player_pawn))
          (if (or _binprogress (Variables|Debug|GetbTestWithoutQuest))
            (CallFunction|ActivateControlTorches)))))))

(event EventPlayerOutOfInteractRange
  (params (Actor :type Actor%20Object%20Reference))
  (|CallParentFunction Actor)
  (Networking|SwitchHasAuthority
    (:Authority
      (CallFunction|ProcessPlayerEndOverlap Actor))
    (:Remote
      (Development|PrintString "@@@@@@ Out Of Interact Range [BP_S18_Command_Symbol]: CALLED!" false (Variables|Debug|GetbDebug))
      (if (!= (Utilities|Array|FindItem (Variables|Data|GetTouchPlayerArray) Actor) -1)
        (bind _asfort_player_pawn (Utilities|Casting|CastToFortPlayerPawn Actor))
        (if (Pawn|IsLocallyControlled _asfort_player_pawn)
          (Utilities|Array|RemoveItem (Variables|Data|GetTouchPlayerArray) Actor)
          (bind _binprogress (CallFunction|CheckQuestProgress _asfort_player_pawn))
          (if (or _binprogress (Variables|Debug|GetbTestWithoutQuest))
            (CallFunction|DeactivateControlTorches)))))))

(event Collision|EventActorBeginOverlap
  (params (OtherActor :type Actor%20Object%20Reference))
  (|CallParentFunction OtherActor) ; This node is disabled and will not be called.\nDrag off pins to build functionality.
)

(event EventTick
  (params (DeltaSeconds :type Float%20%28single-precision%29))
  (|CallParentFunction DeltaSeconds) ; This node is disabled and will not be called.\nDrag off pins to build functionality.
)

(fn ActivateControlTorches
  (for _array_element (Variables|Data|GetControlTorchArray)
    (bind _asbp_s18_control_torch (Utilities|Casting|CastToBP_S18_Control_Torch _array_element))
    (Class|BPS18ControlTorch|SetTorchState _asbp_s18_control_torch "NewEnumerator1" :!class /BattlePassS18/QuestActors/MonsterHunter/BP_S18_Control_Torch.BP_S18_Control_Torch_C))
  (return))

(fn CheckQuestProgress
  (params (NewParam :type Fort%20Player%20Pawn%20Object%20Reference))
  (bind _returnvalue (Fortnite|GetFortPlayerControllerfromActor NewParam))
  (bind _returnvalue_2 (Quests|GetQuestManager _returnvalue "Invalid"))
  (Utilities|IsValid _returnvalue
    (:Is%20Valid
      (Utilities|IsValid _returnvalue_2
        (:Is%20Valid
          (bind _cancompletequest (Class|AthenaQuestFunctionLibrary|AthenaCheckCreativeMode (Game|GetGameState)))
          (if _cancompletequest
            (return (Quest|IsQuestinProgress _returnvalue_2 (Class|ConsolidatedQuestComponent|GetQuestItemDefinition (Variables|BuildingGameplayActorPropQuest|GetConsolidatedQuestComponent))))
            (else
              (return false))))
        (:Is%20Not%20Valid
          (return false))))
    (:Is%20Not%20Valid
      (return false))))

(fn DeactivateControlTorches
  (for _array_element (Variables|Data|GetControlTorchArray)
    (bind _asbp_s18_control_torch (Utilities|Casting|CastToBP_S18_Control_Torch _array_element))
    (Class|BPS18ControlTorch|SetTorchState _asbp_s18_control_torch "NewEnumerator0" :!class /BattlePassS18/QuestActors/MonsterHunter/BP_S18_Control_Torch.BP_S18_Control_Torch_C))
  (return))

(fn EvaluateSymbolReveal
  (params (testPlayer :type Fort%20Player%20Controller%20Object%20Reference))
  (bind _testplayerpawn (LocalVariables|GetTestPlayerPawn))
  (bind _returnvalue (Ability|GetFortAbilitySystemComponent _testplayerpawn))
  (bind _gameplayeffectstostackcnts (Variables|Data|GetGameplayEffectsToStackCnts))
  (LocalVariables|SetTestPlayerPawn (Game|Player|GetPlayerPawn testPlayer))
  (Development|PrintString (Utilities|String|Append "@@@@@@@@@ Evaluate Symbol Reveal [BP_S18_Command_Symbol]: checking " (Utilities|GetDisplayName testPlayer) "...") false (Variables|Debug|GetbDebug))
  (LocalVariables|SetbRevealSymbol true)
  (bind _keys (Utilities|Map|Keys (Variables|Data|GetGameplayEffectsToStackCnts)))
  (bind _array_element (Utilities|Array|ForEachLoopwithBreak _keys)
    (:LoopBody
      (Development|PrintString (Utilities|String|Append "@@@@@@@@@@@@ Evaluate Symbol Reveal [BP_S18_Command_Symbol]: " (Utilities|GetClassDisplayName _array_element) " need " (Utilities|String|ToString(Integer) (Utilities|Map|Find _gameplayeffectstostackcnts _array_element)) " stacks, have " (Utilities|String|ToString(Integer) (GameplayEffects|GetGameplayEffectCount _returnvalue _array_element 0 true)) "...") false (Variables|Debug|GetbDebug))
      (if (not (== (GameplayEffects|GetGameplayEffectCount _returnvalue _array_element 0 true) (Utilities|Map|Find _gameplayeffectstostackcnts _array_element)))
        (LocalVariables|SetbRevealSymbol false)))
    (:Completed
      _))
  (Development|PrintString (Utilities|String|Append "@@@@@@@@@@@@@@@ Evaluate Symbol Reveal [BP_S18_Command_Symbol]: returning " (Utilities|String|ToString(Boolean) (LocalVariables|GetbRevealSymbol)) "!") false (Variables|Debug|GetbDebug))
  (if (LocalVariables|GetbRevealSymbol)
    (Environment|ApplyGlobalEnvironmentGameplayEffecttoActor (LocalVariables|GetTestPlayerPawn) "/BattlePassS18/QuestActors/MonsterHunter/GE_S18_Command_Symbol_Revealed.GE_S18_Command_Symbol_Revealed_C")
    (CallFunction|CheckTagsandComplete (Variables|GettestPlayer))
    (return (LocalVariables|GetbRevealSymbol))
    (else
      (return (LocalVariables|GetbRevealSymbol)))))

(fn InitializeCommandSymbol
  (params (ControlTorches :type Array%20of%20Actor%20Object%20References))
  (bind _controltorches (Variables|GetControlTorches))
  (Development|PrintString (Utilities|String|Append "@@@@@@ Initialize Command Symbol [BP_S18_Command_Symbol]: CALLED - processing " (Utilities|String|ToString(Integer) (Utilities|Array|Length ControlTorches)) " Control Torches...") false (Variables|Debug|GetbDebug))
  (Utilities|Map|Add (Variables|Data|GetGameplayEffectsToStackCnts) "/BattlePassS18/QuestActors/MonsterHunter/GE_S18_Control_Torch_Activated.GE_S18_Control_Torch_Activated_C" (Utilities|Array|Length _controltorches))
  (Development|PrintString (Utilities|String|Append "@@@@@@@@@ Initialize Command Symbol [BP_S18_Command_Symbol]: " (Utilities|GetClassDisplayName "/BattlePassS18/QuestActors/MonsterHunter/GE_S18_Control_Torch_Activated.GE_S18_Control_Torch_Activated_C") " stack count = " (Utilities|String|ToString(Integer) (Utilities|Map|Find (Variables|Data|GetGameplayEffectsToStackCnts) "/BattlePassS18/QuestActors/MonsterHunter/GE_S18_Control_Torch_Activated.GE_S18_Control_Torch_Activated_C"))) false (Variables|Debug|GetbDebug))
  (for _array_element _controltorches
    (bind _asbp_s18_control_torch (Utilities|Casting|CastToBP_S18_Control_Torch _array_element))
    (Variables|/BattlePassS18/QuestActors/MonsterHunter/BP_S18_Control_Torch.BP_S18_Control_Torch_C|SetTorchController self _asbp_s18_control_torch))
  (return))

(fn ProcessPlayerBeginOverlap
  (params (OverlapActor :type Actor%20Object%20Reference))
  (Utilities|IsValid OverlapActor
    (:Is%20Valid
      ;[0.1277 0.8297 1] Add Torch Tag Listener
      (GameplayEffects|ApplyGameplayEffectToTarget :self (Ability|GetAbilitySystemComponent self) :GameplayEffectClass (Variables|Data|GetTorchListenerEffect) :Target (Ability|GetAbilitySystemComponent OverlapActor) :Level 0.0))))

(fn ProcessPlayerEndOverlap
  (params (OverlapActor :type Actor%20Object%20Reference))
  (Utilities|IsValid OverlapActor
    (:Is%20Valid
      ;[0.1277 0.8297 1] Remove Torch Tag Listener
      (GameplayEffects|RemoveActiveGameplayEffectbySourceEffect (Ability|GetAbilitySystemComponent OverlapActor) (Variables|Data|GetTorchListenerEffect) (Ability|GetAbilitySystemComponent self) -1)
      ;[1 0.3781 0.08416] Remove all active Control Torch Effects
      (GameplayEffects|RemoveActiveEffectswithGrantedTags (Ability|GetAbilitySystemComponent (Variables|GetOverlapActor)) "(GameplayTags=((TagName=\"Granted.Quest.ControlTorch.Crouch\"),(TagName=\"Granted.Quest.ControlTorch.Dance\"),(TagName=\"Granted.Quest.ControlTorch.Jump\"),(TagName=\"Granted.Quest.ControlTorch.Pick\"),(TagName=\"Athena.Quests.CommandSymbol.Revealed\"),(TagName=\"Granted.Quest.ControlTorch.Activated\")))")
      ;[0.6099 0.08044 1] Remove exiting player from all Control Torch Interaction Arrays
      (for _array_element (Variables|Data|GetControlTorchArray)
        (bind _asbp_s18_control_torch (Utilities|Casting|CastToBP_S18_Control_Torch _array_element))
        (Utilities|Array|RemoveItem (Variables|/Game/Athena/Items/QuestParents/PropQuestActor/BP_Athena_PropQuestActor_Parent.BP_Athena_PropQuestActor_Parent_C|GetFortPlayerControllersWhoInteracted _asbp_s18_control_torch) (Fortnite|GetFortPlayerControllerfromActor (Variables|GetOverlapActor)))))))

(fn ConstructionScript
  (|CallParentFunction)
  (Development|PrintString "@@@ Construction [BP_S18_Command_Symbol]: CALLED!" false (Variables|Debug|GetbDebug)))
