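; Season 18 Monster Hunter quest torch prop, themed by TorchKind as a crouch, dance, jump or pickaxe challenge with matching Jones hologram, action tag and success effects.
; Entering range applies a listener gameplay effect on the server and, on the local client, reveals the hologram if the torch is in its middle state; leaving reverses both.
; The server ignores repeat interactors, while the local client advances that middle state to completed, turning the light green and hiding the hologram.
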(event EventPlayerInInteractRange
  (params (Actor :type Actor%20Object%20Reference))
  (Variables|Default|SetTestActor Actor)
  (Networking|SwitchHasAuthority
    (:Authority
      (CallFunction|ProcessPlayerBeginOverlap Actor))
    (:Remote
      (bind _asfort_player_pawn (Utilities|Casting|CastToFortPlayerPawn (Variables|Default|GetTestActor)))
      (if (Utilities|Array|ContainsItem (Fortnite|GetLocalFortPlayerControllers true) (Fortnite|GetFortPlayerControllerfromActor _asfort_player_pawn))
        (if (Utilities|Enum|Equal(Enum) (Variables|Torch|GetTorchState) "NewEnumerator1")
          (bind _output_get (Variables|Torch|SetbIsOn true))
          (Development|SetHiddeninGame (Variables|Default|GetSM_JonesHolo) (not _output_get) true)
          (Development|PrintString "@@@ Event Player in Interact Range [BP_S18_Control_Torch]: CALLED!" false (Variables|Debug|GetbDebug)))))))

(event EventPlayerOutOfInteractRange
  (params (Actor :type Actor%20Object%20Reference))
  (Variables|Default|SetTestActor Actor)
  (Networking|SwitchHasAuthority
    (:Authority
      (CallFunction|ProcessPlayerEndOverlap Actor))
    (:Remote
      (bind _asfort_player_pawn (Utilities|Casting|CastToFortPlayerPawn (Variables|Default|GetTestActor)))
      (if (Utilities|Array|ContainsItem (Fortnite|GetLocalFortPlayerControllers true) (Fortnite|GetFortPlayerControllerfromActor _asfort_player_pawn))
        (if (Utilities|Enum|Equal(Enum) (Variables|Torch|GetTorchState) "NewEnumerator1")
          (bind _output_get (Variables|Torch|SetbIsOn false))
          (Development|SetHiddeninGame (Variables|Default|GetSM_JonesHolo) (not _output_get) true)
          (Development|PrintString "@@@ Event Player Out Of Interact Range [BP_S18_Control_Torch]:  CALLED!" false (Variables|Debug|GetbDebug)))))))

(event EventBeginPlay
  (|CallParentFunction)
  (bind _returnvalue (Rendering|Material|CreateDynamicMaterialInstance (Variables|Components|GetSM_Quest_PreInteraction) 0 "/Game/Gadgets/Assets/GlowStick/Material/MI_FORT_Gadgets_Fragment_Container_Inst.MI_FORT_Gadgets_Fragment_Container_Inst" "None" :!class /Script/Engine.PrimitiveComponent))
  (Variables|Default|SetTorch_Mat _returnvalue)
  (CallFunction|SetLightBrightness (Variables|Light|GetLightBrightnessOption))
  (CallFunction|SetLightColor (Variables|Light|GetLightColorOption))
  (CallFunction|SetTorchKind (Variables|Torch|GetTorchKind))
  (CallFunction|SetTorchState "NewEnumerator0"))

(event EventChangeStateForPlayer
  (params (FPC :type Fort%20Player%20Controller%20Object%20Reference))
  (Networking|SwitchHasAuthority
    (:Authority
      (if (Utilities|Array|ContainsItem (Variables|Advanced|Interaction|GetFortPlayerControllersWhoInteracted) FPC)
        (Development|PrintString "@@@@@@@@@ Event Blueprint on Interact [BP_S18_Control_Torch]: Already interacted..." false (Variables|Debug|GetbDebug))
        (else
          (|CallParentFunction FPC))))
    (:Remote
      (if (Pawn|IsLocalPlayerController FPC)
        (if (Utilities|Enum|Equal(Enum) (Variables|Torch|GetTorchState) "NewEnumerator1")
          (|CallParentFunction FPC)
          (Development|PrintString "@@@@@@@@@ Event Blueprint on Interact [BP_S18_Control_Torch]: Interact SUCCESS!!!" false (Variables|Debug|GetbDebug))
          (CallFunction|SetTorchState "NewEnumerator2"))))))

(event Collision|EventActorBeginOverlap
  (params (OtherActor :type Actor%20Object%20Reference))
  (|CallParentFunction OtherActor) ; This node is disabled and will not be called.\nDrag off pins to build functionality.
)

(event EventTick
  (params (DeltaSeconds :type Float%20%28single-precision%29))
  (|CallParentFunction DeltaSeconds) ; This node is disabled and will not be called.\nDrag off pins to build functionality.
)

(fn BlueprintCanInteract
  (params (InteractingPawn :type Fort%20Pawn%20Object%20Reference) (InteractionBeingAttempted :type EInteractionBeingAttempted%20Enum :default "FirstInteraction") (InteractionType :type TInteractionType%20Enum :default "IT_NoInteraction"))
  (bind _asfort_player_controller (Utilities|Casting|CastToFortPlayerController (Pawn|GetController InteractingPawn)))
  (return (or (Variables|Design|QuestSetup|GetbTestVisualsNoQuestRequired) (Math|Boolean|ANDBoolean (Quest|HasCompletedObjectivewithName (Quests|GetQuestManager _asfort_player_controller "Invalid") (Class|ConsolidatedQuestComponent|GetQuestItemDefinition (Variables|BuildingGameplayActorPropQuest|GetConsolidatedQuestComponent)) (Variables|Default|GetObjectiveName) :!class /Script/FortniteGame.FortQuestManager) (|CallParentFunction InteractingPawn InteractionBeingAttempted InteractionType) (Utilities|Enum|Equal(Enum) (Variables|Torch|GetTorchState) "NewEnumerator1")))))

(fn ProcessPlayerBeginOverlap
  (params (OverlapActor :type Actor%20Object%20Reference))
  (Utilities|IsValid OverlapActor
    (:Is%20Valid
      ;[0.1277 0.8297 1] Add Torch Action Listener
      (GameplayEffects|ApplyGameplayEffectToTarget :self (Ability|GetAbilitySystemComponent self) :GameplayEffectClass (LocalVariables|GetTorchActionListenerEffect) :Target (Ability|GetAbilitySystemComponent OverlapActor) :Level 0.0))))

(fn ProcessPlayerEndOverlap
  (params (OverlapActor :type Actor%20Object%20Reference))
  (Utilities|IsValid OverlapActor
    (:Is%20Valid
      ;[0.1277 0.8297 1] Remove Torch Tag Listener
      (GameplayEffects|RemoveActiveGameplayEffectbySourceEffect (Ability|GetAbilitySystemComponent OverlapActor) (LocalVariables|GetTorchActionListenerEffect) (Ability|GetAbilitySystemComponent self) -1))))

(fn SetLightBrightness
  (params (LightBrightness :type Enum_LightBrightness%20Enum :default "NewEnumerator0"))
  (bind _pointlight (Variables|Default|GetPointLight))
  (switch Utilities|FlowControl|Switch|SwitchonEnum_LightBrightness LightBrightness
    (:NewEnumerator0
      (Rendering|Components|Light|SetIntensity _pointlight 5.5))
    (:NewEnumerator1
      (Rendering|Components|Light|SetIntensity _pointlight 11))
    (:NewEnumerator2
      (Rendering|Components|Light|SetIntensity _pointlight 2.75))))

(fn SetLightColor
  (params (Color :type Enum_LightColorsGlobal%20Enum :default "NewEnumerator7"))
  (bind _structlightcolorsglobal (Variables|Default|GetStructLightColorsGlobal))
  (bind (_blue_2_835dee5b4b1e8fe2dcbd2eb808f69637 _teal_4_32feb0134e88f33098310f87b27d1674 _purple_6_d175df4b447eed701e402daeeb0431e4 _yellow_8_47bf9e0c4852ef9111c4ce903b40a8b3 _white_10_2fdb792d4a2a1052ce1b4789ea94fe97 _orange_12_d80329cb4e7f729bbecd1b9b879bb103 _green_14_f93ebbc04e7685314d7338b74c858b52 _red_16_8f0f9b8040677bb0fb52e9bb1fd9a55e) (Utilities|Struct|BreakStructLightColorsGlobal _structlightcolorsglobal))
  (switch Utilities|FlowControl|Switch|SwitchonEnum_LightColorsGlobal Color
    (:NewEnumerator7
      (Variables|Default|SetCurrentColorStructure _white_10_2fdb792d4a2a1052ce1b4789ea94fe97))
    (:NewEnumerator0
      (Variables|Default|SetCurrentColorStructure _blue_2_835dee5b4b1e8fe2dcbd2eb808f69637))
    (:NewEnumerator1
      (Variables|Default|SetCurrentColorStructure _teal_4_32feb0134e88f33098310f87b27d1674))
    (:NewEnumerator2
      (Variables|Default|SetCurrentColorStructure _purple_6_d175df4b447eed701e402daeeb0431e4))
    (:NewEnumerator3
      (Variables|Default|SetCurrentColorStructure _yellow_8_47bf9e0c4852ef9111c4ce903b40a8b3))
    (:NewEnumerator4
      (Variables|Default|SetCurrentColorStructure _orange_12_d80329cb4e7f729bbecd1b9b879bb103))
    (:NewEnumerator5
      (Variables|Default|SetCurrentColorStructure _green_14_f93ebbc04e7685314d7338b74c858b52))
    (:NewEnumerator6
      (Variables|Default|SetCurrentColorStructure _red_16_8f0f9b8040677bb0fb52e9bb1fd9a55e)))
  (Rendering|Components|Light|SetLightColor (Variables|Default|GetPointLight) (Variables|Default|GetCurrentColorStructure) true)
  (Effects|Components|ParticleSystem|SetColorParameter (Variables|Default|GetNS_GlowTorchAlt) "Color" (Variables|Default|GetCurrentColorStructure))
  (Rendering|Material|SetVectorParameterValue (Variables|Default|GetTorch_Mat) "EmissiveMultiplier" (Variables|Default|GetCurrentColorStructure) :!class /Script/Engine.MaterialInstanceDynamic))

(fn SetTorchKind
  (params (NewTorchKind :type TorchKindEnum%20Enum :default "NewEnumerator0"))
  (switch Utilities|FlowControl|Switch|SwitchonTorchKindEnum NewTorchKind
    (:NewEnumerator0
      (LocalVariables|SetNewHoloMesh "/Game/Environments/Apollo/Props/Quests/Props/Meshes/JonesCrouch.JonesCrouch")
      (Variables|Design|QuestSetup|InteractionSettings|SetDamage%20or%20Interact%20Tag "(GameplayTags=((TagName=\"Gameplay.Action.Player.Crouch\")))")
      (LocalVariables|SetHoloMeshScale "0.750000,0.750000,0.750000")
      (LocalVariables|SetConeMeshScale "0.500000,0.500000,0.500000")
      (LocalVariables|SetConeLoc "0, 0,-62.000000")
      (Variables|Torch|SetTorchSuccessEffect "/BattlePassS18/QuestActors/MonsterHunter/GE_S18_Control_Torch_Crouch.GE_S18_Control_Torch_Crouch_C"))
    (:NewEnumerator1
      (LocalVariables|SetNewHoloMesh "/Game/Environments/Apollo/Props/Quests/Props/Meshes/JonesDance.JonesDance")
      (Variables|Design|QuestSetup|InteractionSettings|SetDamage%20or%20Interact%20Tag "(GameplayTags=((TagName=\"Gameplay.Action.Player.Emote\")))")
      (LocalVariables|SetHoloMeshScale "1.500000,1.500000,1.500000")
      (LocalVariables|SetConeMeshScale "0.500000,0.500000,0.500000")
      (LocalVariables|SetConeLoc "0, 0,-30.000000")
      (Variables|Torch|SetTorchSuccessEffect "/BattlePassS18/QuestActors/MonsterHunter/GE_S18_Control_Torch_Dance.GE_S18_Control_Torch_Dance_C"))
    (:NewEnumerator2
      (LocalVariables|SetNewHoloMesh "/Game/Environments/Apollo/Props/Quests/Props/Meshes/JonesJump.JonesJump")
      (Variables|Design|QuestSetup|InteractionSettings|SetDamage%20or%20Interact%20Tag "(GameplayTags=((TagName=\"Event.Physics.Landed\")))")
      (LocalVariables|SetHoloMeshScale "1.500000,1.500000,1.500000")
      (LocalVariables|SetConeMeshScale "0.500000,0.500000,0.500000")
      (LocalVariables|SetConeLoc "0, 0,-30.000000")
      (Variables|Torch|SetTorchSuccessEffect "/BattlePassS18/QuestActors/MonsterHunter/GE_S18_Control_Torch_Jump.GE_S18_Control_Torch_Jump_C"))
    (:NewEnumerator3
      (LocalVariables|SetNewHoloMesh "/Game/Environments/Apollo/Props/Quests/Props/Meshes/JonesPick.JonesPick")
      (Variables|Design|QuestSetup|InteractionSettings|SetDamage%20or%20Interact%20Tag "(GameplayTags=((TagName=\"Gameplay.Action.Attacking.Melee\")))")
      (LocalVariables|SetHoloMeshScale "1.500000,1.500000,1.500000")
      (LocalVariables|SetConeMeshScale "0.500000,0.500000,0.500000")
      (LocalVariables|SetConeLoc "0, 0,-30.000000")
      (Variables|Torch|SetTorchSuccessEffect "/BattlePassS18/QuestActors/MonsterHunter/GE_S18_Control_Torch_Pick.GE_S18_Control_Torch_Pick_C")))
  (Components|StaticMesh|SetStaticMesh (Variables|Default|GetSM_JonesHolo) (LocalVariables|GetNewHoloMesh))
  (Transformation|SetWorldScale3D (Variables|Default|GetSM_JonesHolo) (LocalVariables|GetHoloMeshScale))
  (Transformation|SetWorldScale3D (Variables|Default|GetSM_LockHologramCone) (LocalVariables|GetConeMeshScale))
  (Transformation|SetRelativeLocation (Variables|Default|GetSM_LockHologramCone) (LocalVariables|GetConeLoc) false false)
  (Variables|Design|QuestSetup|GameplayTags,EffectsandCosmetics|SetInteraction%20GEs (Utilities|Array|MakeArray (Variables|Torch|GetTorchSuccessEffect) "/BattlePassS18/QuestActors/MonsterHunter/GE_S18_Control_Torch_Activated.GE_S18_Control_Torch_Activated_C")))

(fn SetTorchState
  (params (New%20State :type TorchStateEnum%20Enum :default "NewEnumerator0"))
  (Development|PrintString (Utilities|String|Append "@@@@@@ SetTorchState [BP_S18_Control_Torch]: CALLED, setting Torch to " (Utilities|String|EnumtoString New%20State) "...") false (Variables|Debug|GetbDebug))
  (bind _output_get (Variables|Torch|SetTorchState (Variables|GetNew%20State)))
  (switch Utilities|FlowControl|Switch|SwitchonTorchStateEnum _output_get
    (:NewEnumerator0
      (Variables|Torch|SetbIsOn false)
      (Development|SetHiddeninGame (Variables|Default|GetSM_JonesHolo) true true)
      (CallFunction|SetLightBrightness "NewEnumerator2")
      (CallFunction|SetLightColor "NewEnumerator0"))
    (:NewEnumerator1
      (CallFunction|SetLightBrightness "NewEnumerator0"))
    (:NewEnumerator2
      (CallFunction|SetLightColor "NewEnumerator5")
      (Variables|Torch|SetbIsOn false)
      (Development|SetHiddeninGame (Variables|Default|GetSM_JonesHolo) true true))))

(fn ConstructionScript
  (|CallParentFunction)
  (bind _returnvalue (Rendering|Material|CreateDynamicMaterialInstance (Variables|Components|GetSM_Quest_PreInteraction) 0 "/Game/Gadgets/Assets/GlowStick/Material/MI_FORT_Gadgets_Fragment_Container_Inst.MI_FORT_Gadgets_Fragment_Container_Inst" "None" :!class /Script/Engine.PrimitiveComponent))
  (Variables|Default|SetTorch_Mat _returnvalue)
  (CallFunction|SetLightBrightness (Variables|Light|GetLightBrightnessOption))
  (CallFunction|SetLightColor (Variables|Light|GetLightColorOption))
  (CallFunction|SetTorchKind (Variables|Torch|GetTorchKind))
  (CallFunction|SetTorchState "NewEnumerator0"))
