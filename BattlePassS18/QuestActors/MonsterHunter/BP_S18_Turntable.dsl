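; Season 18 Battle Pass Monster Hunter turntable quest prop, with quest-gated interaction and linked-actor destruction handling.
; Interaction requires a test-visuals override, or both a completed named quest objective and the parent class check.
; When its linked actor is destroyed, it sets bDestroyedState, disables collision and destroys itself.
; The bDestroyedState RepNotify handles hiding the table mesh and activating the Niagara destruction effect.
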(event Custom|OnLinkedActorDestroyed
  ;[0.4589 0.1573 1] Handle Bound Actor destruction processing
  (Variables|Default|SetbDestroyedState true)
  (Collision|SetActorEnableCollision false)
  (Actor|DestroyActor))

(event OnLinkedActorDestroyed(FortLinkToActor)
  (params (DamageCauser :type Actor%20Object%20Reference))
  ;[0.4589 0.1573 1] Handle Bound Actor destruction processing
  (CallFunction|OnLinkedActorDestroyed))

(fn BlueprintCanInteract
  (params (InteractingPawn :type Fort%20Pawn%20Object%20Reference) (InteractionBeingAttempted :type EInteractionBeingAttempted%20Enum :default "FirstInteraction") (InteractionType :type TInteractionType%20Enum :default "IT_NoInteraction"))
  (bind _asfort_player_controller (Utilities|Casting|CastToFortPlayerController (Pawn|GetController InteractingPawn)))
  (return (or (Variables|Design|QuestSetup|GetbTestVisualsNoQuestRequired) (and (Quest|HasCompletedObjectivewithName (Quests|GetQuestManager _asfort_player_controller "Invalid") (Class|ConsolidatedQuestComponent|GetQuestItemDefinition (Variables|BuildingGameplayActorPropQuest|GetConsolidatedQuestComponent)) (LocalVariables|GetObjectiveName) :!class /Script/FortniteGame.FortQuestManager) (|CallParentFunction InteractingPawn InteractionBeingAttempted InteractionType)))))

(fn OnRep_bDestroyedState
  (if (Variables|Default|GetbDestroyedState)
    (Development|SetHiddeninGame (Variables|Default|GetSM_Table) true false)
    (Components|Activation|Activate (Variables|Default|GetNS_General_Deco_Destruction_Large_Converted) false)))

(fn ConstructionScript
  (|CallParentFunction))
