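; Event-triggered gameplay ability for the Monster Hunter quest that reacts to a torch-related tag on a player pawn and updates a Command Symbol.
; On activation it stores the event tag and finds the symbol through the effect causer of the effect that granted the ability.
; It then removes active effects granting that tag from the player, has the symbol evaluate its reveal for that player's controller, and ends.
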(event Ability|EventActivateAbilityFromEvent
  (params (EventData :type Gameplay%20Event%20Data%20Structure%20%28by%20ref%29))
  (|CallParentFunction EventData)
  (Variables|Default|SetEvent%20Tag (Utilities|Struct|BreakGameplayEventData EventData))
  (bind _asfort_player_pawn (Utilities|Casting|CastToFortPlayerPawn (Ability|GetAvatarActorfromActorInfo self))
    (:then
      (bind _asbp_s18_command_symbol (Utilities|Casting|CastToBP_S18_Command_Symbol (Ability|EffectContext|GetEffectCauser (Ability|GetGrantedbyEffectContext)))
        (:then
          (GameplayEffects|RemoveActiveEffectswithGrantedTags (Ability|GetAbilitySystemComponent _asfort_player_pawn) (GameplayTags|MakeGameplayTagContainerfromTag (Variables|Default|GetEvent%20Tag)))
          (Class|BPS18CommandSymbol|EvaluateSymbolReveal _asbp_s18_command_symbol (Fortnite|GetFortPlayerControllerfromActor _asfort_player_pawn))
          (Ability|EndAbility))
        (:CastFailed
          (Ability|EndAbility))))
    (:CastFailed
      (Ability|EndAbility))))
